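# Physics asset for the Feline Warrior back bling, setting up simulated motion for its sword, bow pieces and bow strings.
# The sword, two bow bodies and two four-link string chains all hang from a base joint through angular-limited constraints.
# The sword is held tightly with small limits, while the string capsules swing more freely with little twist.
# A shoulder capsule and large spine box have no constraints, apparently serving as collision volumes.
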PhysicsAsset /BRCosmetics/Accessories/FORT_Backpacks/Backpack_F_MED_FelineWarrior/Meshes/F_MED_FelineWarrior_Pack_Physics
bodies {
 dyn_backpack_shoulder_joint [{
   shapeType=Capsule
   length=11
   radius=6
   shapeName=dyn_backpack_shoulder_joint_capsule
   center=<-0.7427 1.5404 0>
   rotation=<rot 90 179.9999 -179.9997>
  }]
 dyn_backpack_spine_joint [{shapeType=Box, extentX=13.5, extentY=19.5, extentZ=26.5, shapeName=dyn_backpack_spine_joint_box}]
 dyn_base_joint [{shapeType=Capsule, length=1, radius=0.5, shapeName=dyn_base_joint_capsule}]
 dyn_bow_1_joint [{shapeType=Capsule, length=3, radius=1, shapeName=dyn_bow_1_joint_capsule, center=<3.9451 0 0>, rotation=<rot -90 179.9999 179.9998>}]
 dyn_bow_2_joint [{shapeType=Capsule, length=3, radius=1, shapeName=dyn_bow_2_joint_capsule, center=<4.163 0 0>, rotation=<rot 90 0 0.0005>}]
 dyn_bow_string_1_joint [{shapeType=Capsule, length=1, radius=0.5, shapeName=dyn_bow_string_1_joint_capsule, center=<1.5654 0 0>, rotation=<rot 90 179.9802 179.9803>}]
 dyn_bow_string_2_joint [{shapeType=Capsule, length=1, radius=0.5, shapeName=dyn_bow_string_2_joint_capsule, center=<1.8244 0 0>, rotation=<rot 90 90 90.0002>}]
 dyn_bow_string_3_joint [{shapeType=Capsule, length=1, radius=0.5, shapeName=dyn_bow_string_3_joint_capsule, center=<1.4384 0 0>, rotation=<rot 90 179.9208 179.921>}]
 dyn_bow_string_4_joint [{shapeType=Capsule, length=1, radius=0.5, shapeName=dyn_bow_string_4_joint_capsule, center=<1.7787 0 0>, rotation=<rot 90 179.6835 179.6835>}]
 dyn_bow_string_5_joint [{
   shapeType=Capsule
   length=1.5
   radius=0.5
   shapeName=dyn_bow_string_5_joint_capsule
   center=<1.6034 0 0>
   rotation=<rot 90 179.9802 179.9802>
  }]
 dyn_bow_string_6_joint [{shapeType=Capsule, length=1, radius=0.5, shapeName=dyn_bow_string_6_joint_capsule, center=<1.1535 0 0>, rotation=<rot 90 179.9604 179.9604>}]
 dyn_bow_string_7_joint [{shapeType=Capsule, length=0.5, radius=0.5, shapeName=dyn_bow_string_7_joint_capsule, center=<1.1115 0 0>, rotation=<rot 90 0 0>}]
 dyn_bow_string_8_joint [{shapeType=Capsule, length=1, radius=0.5, shapeName=dyn_bow_string_8_joint_capsule, center=<1.4368 0 0>, rotation=<rot 90 94.7598 94.7596>}]
 dyn_sword_joint [{shapeType=Capsule, length=41, radius=2, shapeName=dyn_sword_joint_capsule, center=<4.5103 0 0>, rotation=<rot -90 165.1488 -165.1494>}]
}
constraints [
 {bone1Name=dyn_sword_joint, bone2Name=dyn_base_joint, swing1LimitDegrees=3, swing2LimitDegrees=5, twistLimitDegrees=2}
 {bone1Name=dyn_bow_1_joint, bone2Name=dyn_base_joint, swing1LimitDegrees=20, swing2LimitDegrees=10, twistLimitDegrees=0}
 {bone1Name=dyn_bow_2_joint, bone2Name=dyn_base_joint, swing1LimitDegrees=20, swing2LimitDegrees=10, twistLimitDegrees=0}
 {bone1Name=dyn_bow_string_1_joint, bone2Name=dyn_base_joint, swing1LimitDegrees=10, swing2LimitDegrees=30, twistLimitDegrees=3}
 {bone1Name=dyn_bow_string_5_joint, bone2Name=dyn_base_joint, swing1LimitDegrees=30, swing2LimitDegrees=30, twistLimitDegrees=3}
 {bone1Name=dyn_bow_string_2_joint, bone2Name=dyn_bow_string_1_joint, swing1LimitDegrees=10, swing2LimitDegrees=45, twistLimitDegrees=3}
 {bone1Name=dyn_bow_string_3_joint, bone2Name=dyn_bow_string_2_joint, swing1LimitDegrees=10, swing2LimitDegrees=45, twistLimitDegrees=3}
 {bone1Name=dyn_bow_string_4_joint, bone2Name=dyn_bow_string_3_joint, swing1LimitDegrees=10, swing2LimitDegrees=45, twistLimitDegrees=3}
 {bone1Name=dyn_bow_string_6_joint, bone2Name=dyn_bow_string_5_joint, swing1LimitDegrees=30, swing2LimitDegrees=30, twistLimitDegrees=3}
 {bone1Name=dyn_bow_string_7_joint, bone2Name=dyn_bow_string_6_joint, swing1LimitDegrees=30, swing2LimitDegrees=30, twistLimitDegrees=3}
 {bone1Name=dyn_bow_string_8_joint, bone2Name=dyn_bow_string_7_joint, swing1LimitDegrees=30, swing2LimitDegrees=30, twistLimitDegrees=3}
]
