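# Animation Blueprint for the Feline Warrior backpack that pins it to the wearer and adds secondary physics motion.
# Each update reads world transforms of the attach parent's spine_05 and upperarm_r sockets.
# Transform Bone nodes apply them in world space to the pack's spine and shoulder dynamic joints, feeding a RigidBody simulation with its own physics asset.
# The simulation's alpha is halved in the front end and full otherwise.
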AnimBlueprint /BRCosmetics/Accessories/FORT_Backpacks/Backpack_F_MED_FelineWarrior/Meshes/F_MED_FelineWarrior_Pack_AnimBP
target_skeleton=&/F_MED_FelineWarrior_Pack_Skeleton
graphs [
 AnimGraph {nodes [
   n1 = Get!() -> (shoulder_Location:Vector, shoulder_Rotation:Rotator, shoulder_Scale:Vector)
   n2 = "Transform (Modify) Bone - Bone: dyn_backpack_shoulder_joint"!(
    ComponentPose=Local To Component(LocalPose=Input Pose())
    bAlphaBoolEnabled:bool=True
    Translation=@n1
    Rotation=@n1.shoulder_Rotation
    Scale=@n1.shoulder_Scale
   ) -> (Pose) {
    alphaInputType=Bool
    rotationMode=BMM_Replace
    rotationSpace=BCS_WorldSpace
    scaleMode=BMM_Replace
    scaleSpace=BCS_WorldSpace
    translationMode=BMM_Replace
    translationSpace=BCS_WorldSpace
    alphaBoolBlend {blendOption=Linear}
    boneToModify {boneName=dyn_backpack_shoulder_joint}
   }
   n3 = Get!() -> (spine_Location:Vector, spine_Rotation:Rotator, spine_Scale:Vector)
   n4 = "Transform (Modify) Bone - Bone: dyn_backpack_spine_joint"!(ComponentPose=@n2, bAlphaBoolEnabled:bool=True, Translation=@n3, Rotation=@n3.spine_Rotation, Scale=@n3.spine_Scale) -> (Pose) {
    alphaInputType=Bool
    rotationMode=BMM_Replace
    rotationSpace=BCS_WorldSpace
    scaleMode=BMM_Replace
    scaleSpace=BCS_WorldSpace
    translationMode=BMM_Replace
    translationSpace=BCS_WorldSpace
    alphaBoolBlend {blendOption=Linear}
    boneToModify {boneName=dyn_backpack_spine_joint}
   }
   n5 = RigidBody(ComponentPose=@n4, Alpha=Get Alpha() -> (Alpha:float)) {
    overridePhysicsAsset=&/F_MED_FelineWarrior_Pack_Physics
    alphaBoolBlend {blendOption=Linear}
    componentAppliedLinearAccClamp=<3000 3000 3000>
    componentLinearAccScale=<0.4 0.4 0.4>
    componentLinearVelScale=<0.4 0.4 0.4>
    simSpaceSettings {maxAngularAcceleration=3000, maxAngularVelocity=3000, maxLinearAcceleration=3000, maxLinearVelocity=3000, worldAlpha=0.3}
   }
   Output Pose(Result=Component To Local(ComponentPose=@n5)) {layerGroup=None}
  ]}
 EventGraph {nodes [
   Event BlueprintUpdateAnimation()
   n1 = GetAttachParent!(self:Scene Component Object Reference=GetPartSkeletalMeshComponent()) -> (ReturnValue:Scene Component Object Reference)
   > Set spine(spine:Transform=GetSocketTransform(self=@n1, InSocketName=spine_05, TransformSpace=RTS_World)) -> (Output_Get:Transform)
   > Set shoulder(shoulder:Transform=GetSocketTransform(self=@n1, InSocketName=upperarm_r, TransformSpace=RTS_World)) -> (Output_Get:Transform)
   > n2 = Branch(Condition=Get bIsInFrontEnd())
   > Set Alpha(Alpha:float=0.5) -> (Output_Get:float)
   Set Alpha(execute=@n2.else, Alpha=1)
   TryGetPawnOwner()
  ]}
]
variables [
 Alpha {type=number, default=0.5}
 shoulder {type=Transform, default {location=<0 0 0>, rotation=<rot 0 0 0>, scale=<1 1 1>}}
 spine {type=Transform, default {location=<0 0 0>, rotation=<rot 0 0 0>, scale=<1 1 1>}}
]
